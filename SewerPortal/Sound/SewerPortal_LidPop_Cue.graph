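# Sound cue for the sewer portal's lid pop, combining a lid-pop sound with a water splash.
# A mixer layers one of three randomly picked lid-pop waves, passed through a modulator, with one of four random sewer splashes at half input volume.
# It uses trap sound class and attenuation, doubles overall volume, and primes on load.
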SoundCue /SewerPortal/Sound/SewerPortal_LidPop_Cue
root=Mixer_0
nodes [
 {id=Mixer_0, children [Modulator_1,Random_1], properties {inputVolume [1,0.5]}}
 {id=Modulator_1, children [Random_0]}
 {id=Random_0, children [WavePlayer_4,WavePlayer_5,WavePlayer_6], properties {weights [1,1,1]}}
 {id=Random_1, children [WavePlayer_7, WavePlayer_8, WavePlayer_9, WavePlayer_10], properties {weights [1,1,1,1]}}
 {id=WavePlayer_10, properties {soundWaveAssetPtr=&/SewerSplash_04}}
 {id=WavePlayer_4, properties {soundWaveAssetPtr=&/SewerPortal_LidPop_01}}
 {id=WavePlayer_5, properties {soundWaveAssetPtr=&/SewerPortal_LidPop_02}}
 {id=WavePlayer_6, properties {soundWaveAssetPtr=&/SewerPortal_LidPop_03}}
 {id=WavePlayer_7, properties {soundWaveAssetPtr=&/SewerSplash_01}}
 {id=WavePlayer_8, properties {soundWaveAssetPtr=&/SewerSplash_02}}
 {id=WavePlayer_9, properties {soundWaveAssetPtr=&/SewerSplash_03}}
]
properties {
 attenuationSettings=&Fort_Traps_Att
 bPrimeOnLoad=true
 maxDistance=6000
 soundClassObject=&Fort_Traps
 volumeMultiplier=2
}
