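# Sound cue that plays a water splash for the Sewer Portal.
# A Random node picks one of four equally weighted SewerSplash waves and feeds it through a Modulator at the root.
# It uses the Fort_Traps sound class and trap attenuation, primes on load, and has a boosted volume multiplier.
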SoundCue /SewerPortal/Sound/SewerPortal_WaterSplash_Cue
root=Modulator_1
nodes [
 {id=Modulator_1, children [Random_0]}
 {id=Random_0, children [WavePlayer_0,WavePlayer_1,WavePlayer_3,WavePlayer_4], properties {weights [1,1,1,1]}}
 {id=WavePlayer_0, properties {soundWaveAssetPtr=&/SewerSplash_01}}
 {id=WavePlayer_1, properties {soundWaveAssetPtr=&/SewerSplash_02}}
 {id=WavePlayer_3, properties {soundWaveAssetPtr=&/SewerSplash_03}}
 {id=WavePlayer_4, properties {soundWaveAssetPtr=&/SewerSplash_04}}
]
properties {
 attenuationSettings=&Fort_Traps_Att
 bPrimeOnLoad=true
 maxDistance=6000
 soundClassObject=&Fort_Traps
 volumeMultiplier=1.3
}
